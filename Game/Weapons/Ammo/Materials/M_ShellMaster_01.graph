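# M_ShellMaster_01 is a lit master material for ammo shells, built for meshes and mesh particles, with texture-driven color, normals, roughness and metallic.
# Emissive combines a diffuse tint, a UV1 gradient and Dynamic Parameter red, fading over scaled particle relative time before MF_FX_EmissiveAdjustments.
# The IsShotgunShellGold static switch recolors a UV region, masked by a rectangle plus a sphere mask, through a three-color gold ramp driven by diffuse brightness.
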Material /Game/Weapons/Ammo/Materials/M_ShellMaster_01
graph {
 properties {
  bUsedWithMeshDeformer=true
  bUsedWithMeshParticles=true
  bUsedWithNanite=true
  bUsedWithNiagaraMeshParticles=true
  bUsedWithSkeletalMesh=true
  pixelDepthOffsetMode=PDOM_Legacy
  usedShadingModels=Default Lit
 }
 nodes [
  Constant() {r=0.5}
  n1 = TextureSampleParameter2D() {group=Specular, parameterName=SpecularMasks, samplerType=SAMPLERTYPE_Masks, texture=/Game/Weapons/Ammo/Textures/T_AmmoSet02_S}
  LinearInterpolate(
   A=ScalarParameter() {defaultValue=0.22, group=Specular, parameterName=SpecRoughnessMin, sliderMax=1}
   B=ScalarParameter() {defaultValue=0.66, group=Specular, parameterName=SpecRoughnessMax, sliderMax=1}
   Alpha=@n1.B
  )
  n2 = TextureSampleParameter2D() {parameterName=Diffuse, texture=/Game/Weapons/Ammo/Textures/T_AmmoSet02_D}
  n3 = TextureCoordinate() {coordinateIndex=1}
  n4 = Multiply(
   A=LinearInterpolate(B=@n2.RGB) {constA=1}
   B=Power(Base=ComponentMask(None=@n3) {g=true}, Exp=ScalarParameter() {defaultValue=2, parameterName=EmissiveGradientExponent})
  )
  n5 = DynamicParameter() {defaultValue=<rgba 0 0 0 1>, paramNames [R,G,B,Param4]}
  n6 = MaterialFunctionCall(R=@n5.R, G=@n5.R, B=@n5.R) {materialFunction=&AppendMany}
  n7 = Multiply(A=ScalarParameter() {defaultValue=3, parameterName=PRT Scale}, B=ParticleRelativeTime())
  MaterialFunctionCall(Emissive=LinearInterpolate(A=Multiply(A=@n4, B=@n6.RGB), Alpha=Clamp(None=@n7)) {constB=0}) {materialFunction=&MF_FX_EmissiveAdjustments}
  Multiply(A=@n1.G, B=ScalarParameter() {defaultValue=1, group=Specular, parameterName=Metallic, sliderMax=1})
  n8 = Constant3Vector() {constant=<rgba 0.1153 0.0894 0.0028 1>}
  n9 = Constant3Vector() {constant=<rgba 0.2686 0.1433 0.0322 1>}
  n10 = Constant3Vector() {constant=<rgba 1 0.7378 0.189 1>}
  n11 = MaterialFunctionCall() {materialFunction=&LinearGradient}
  n12 = Constant() {r=1}
  n13 = Constant()
  n14 = If(A=@n11.UGradient, B=ScalarParameter() {parameterName=X1}, "A > B"=@n12, "A < B"=@n13)
  n15 = If(A=@n11.UGradient, B=ScalarParameter() {defaultValue=0.372, parameterName=X2}, "A > B"=@n12, "A < B"=@n13)
  n16 = If(A=@n11.VGradient, B=ScalarParameter() {defaultValue=0.743, parameterName=Y1}, "A > B"=@n12, "A < B"=@n13)
  n17 = If(A=@n11.VGradient, B=ScalarParameter() {defaultValue=0.827, parameterName=Y2}, "A > B"=@n12, "A < B"=@n13)
  n18 = Add(
   A=SphereMask(A=TextureCoordinate(), B=Constant2Vector() {g=0.655, r=0.945}) {attenuationRadius=0.0625}
   B=Multiply(A=Subtract(A=@n14, B=@n15), B=Subtract(A=@n16, B=@n17))
  )
  n19 = Multiply(A=@n18, B=@n2.RGB)
  n20 = Saturate(None=Multiply(A=Power(Base=@n19) {constExponent=1.5}) {constB=4})
  StaticSwitchParameter(
   True=LinearInterpolate(
    A=@n2.RGB
    B=MaterialFunctionCall(A=@n8, B=@n9, C=@n10, Alpha=@n20) {materialFunction=&Lerp_3Color}.Result
    Alpha=@n18
   )
   False=@n2.RGB
  ) {parameterName=IsShotgunShellGold}
  TextureSampleParameter2D() {parameterName=Normals, samplerType=SAMPLERTYPE_Normal, texture=/Game/Weapons/Ammo/Textures/T_AmmoSet02_N}
  VectorParameter() {parameterName=DebugColor, defaultValue=<rgba 2 1.1943 0.4902 1>}
 ]
}
